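; Blueprint function library that spawns the frontend UI for bulk cosmetic downloads.
; CreateCosmeticDownloadPopup pushes a modal dialog, backed by a newly constructed dialog view model, only when the download system view model reports bulk cosmetics neither installed nor installing.
; CreateCosmeticDownloadWidgets adds download and info widgets to a supplied panel, setting alignment and bottom padding through vertical box slots, and returns both.
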(fn CreateCosmeticDownloadPopup
  (params (ContainerWidget :type User%20Widget%20Object%20Reference) (FortDownloadSystemVM :type Fort%20Download%20System%20VM%20Object%20Reference))
  (bind _containerwidget (Variables|GetContainerWidget))
  (if (and (not (Class|FortDownloadSystemVM|GetIsBulkCosmeticInstalled (Variables|GetFortDownloadSystemVM))) (not (Class|FortDownloadSystemVM|GetIsBulkCosmeticInstalling (Variables|GetFortDownloadSystemVM))))
    (bind _returnvalue (Game|ConstructObjectfromClass "/FortniteCoreUI/UI/Frontend/CosmeticDownload/DialogVM_CosmeticDownload.DialogVM_CosmeticDownload_C" _containerwidget))
    (bind _returnvalue_2 (UserInterface|CreateWidget (Widget|GetOwningPlayer _containerwidget) "/FortniteCoreUI/UI/Frontend/CosmeticDownload/WBP_DialogModal_CosmeticDownload.WBP_DialogModal_CosmeticDownload_C" (Widget|GetOwningPlayer _containerwidget)))
    (Class|WBPDialogModalCosmeticDownload|InitDialog _returnvalue_2 _returnvalue)
    (FortUIManagement|TryPushWidgetasModalPanel (Variables|GetContainerWidget) _returnvalue_2)))

(fn CreateCosmeticDownloadWidgets
  (params (ContainerWidget :type Panel%20Widget%20Object%20Reference) (HorizontalAlignment :type EHorizontalAlignment%20Enum :default "HAlign_Fill"))
  (Utilities|FlowControl|Sequence
    (:then_0
      (bind _returnvalue (UserInterface|CreateWidget (Variables|GetContainerWidget) "/FortniteCoreUI/UI/Frontend/CosmeticDownload/WBP_CosmeticDownload.WBP_CosmeticDownload_C" (Widget|GetOwningPlayer (Variables|GetContainerWidget))))
      (Widget|Panel|AddChild (Variables|GetContainerWidget) _returnvalue)
      (Layout|VerticalBoxSlot|SetHorizontalAlignment (Slot|SlotasVerticalBoxSlot _returnvalue) (Variables|GetHorizontalAlignment))
      (Layout|VerticalBoxSlot|SetPadding (Slot|SlotasVerticalBoxSlot _returnvalue) (struct :Left 0.0 :Top 0.0 :Right 0.0 :Bottom 1e+01)))
    (:then_1
      (bind _returnvalue_2 (UserInterface|CreateWidget (Variables|GetContainerWidget) "/FortniteCoreUI/UI/Frontend/CosmeticDownload/WBP_CosmeticDownloadInfo.WBP_CosmeticDownloadInfo_C" (Widget|GetOwningPlayer (Variables|GetContainerWidget))))
      (Widget|Panel|AddChild (Variables|GetContainerWidget) _returnvalue_2)
      (Layout|VerticalBoxSlot|SetHorizontalAlignment (Slot|SlotasVerticalBoxSlot _returnvalue_2) (Variables|GetHorizontalAlignment))
      (Layout|VerticalBoxSlot|SetPadding (Slot|SlotasVerticalBoxSlot _returnvalue_2) (struct :Left 0.0 :Top 0.0 :Right 0.0 :Bottom 1e+01))
      (return _returnvalue _returnvalue_2))))
