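# Frontend cosmetic download indicator widget that shows "Downloading Items" beside a download progress widget over a dark UIKit backplate, with a post-buffer capture widget in the same overlay.
# On Construct it applies the content-on-demand font material to the text, then calls SetCompactUIStyle, whose tablet aspect ratio check has no connected actions.
# SetDownloadVisibility toggles the widget between Collapsed and SelfHitTestInvisible.
# OnUpdateDownloadProgress is an empty function here.
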WidgetBlueprint /FortniteCoreUI/UI/Frontend/CosmeticDownload/WBP_CosmeticDownload
graphs [
 EventGraph {nodes [
   Event Construct()
   > SetFontMaterial!(
    execute
    self:Text Object Reference=Get CommonText_Downloading() -> (CommonText_Downloading:Common Text Object Reference)
    InMaterial:Material Interface Object Reference=/Game/UI/ContentOnDemand/MI_UI_ContentOnDemand_FontMaterial
   ) -> (then)
   > SetCompactUIStyle() {is=title K2Node_CallFunction}
  ]}
 OnUpdateDownloadProgress {nodes [OnUpdateDownloadProgress() {is=title K2Node_FunctionEntry}]}
 SetCompactUIStyle {nodes [
   SetCompactUIStyle() {is=title K2Node_FunctionEntry}
   > Branch(Condition=IsTabletAspectRatio!() -> (ReturnValue:bool))
  ]}
 SetDownloadVisibility {nodes [
   n1 = SetDownloadVisibility!() -> (then, ShouldBeVisible:bool=~) {is=title K2Node_FunctionEntry}
   > SetVisibility(InVisibility=Select(Option 0=Collapsed, Option 1=SelfHitTestInvisible, Index=@n1))
  ]}
]
variables [
 HasForcedStoppedAnimation {type=boolean}
 OnUpdateDownloadProgress {}
]
widgets {
 info {inheritedWidgetCount=0, parentClass=&UserWidget, rootWidgetClass=&Overlay, widgetCount=5}
 widgets [
  Overlay_37 {widgetClassPath=&Overlay, widget=Overlay_37}
  WBP_CaptureForPostBufferUpdate {
   widgetClassPath=&WBP_CaptureForPostBufferUpdate
   bIsVariable=true
   parent=Overlay_37
   slot {class=OverlaySlot, properties {horizontalAlignment=HAlign_Fill, verticalAlignment=VAlign_Fill}}
  }
  UIKit_Backplate {
   widgetClassPath=&WBP_UIKit_Backplate
   parent=Overlay_37
   widget=UIKit_Backplate
   properties {brightness=Dark, renderTransformPivot=<1 0.5>}
   slot {class=OverlaySlot, properties {horizontalAlignment=HAlign_Fill, verticalAlignment=VAlign_Fill}}
  }
  CommonText_Downloading {
   widgetClassPath=&CommonTextBlock
   bIsVariable=true
   parent=Overlay_37
   widget=CommonText_Downloading
   properties {
    justification=Center
    renderOpacity=0.75
    style=&TextStyle-Base_C
    text=Downloading Items
    font {
     bForceMonospaced=false
     bMaterialIsStencil=false
     fontMaterial=null
     fontObject=&NotoSans
     letterSpacing=0
     monospacedWidth=1
     size=24
     skewAmount=0
     typefaceFontName=NotoSans-Regular
     outlineSettings {
      bApplyOutlineToDropShadows=false
      bMiteredCorners=false
      bSeparateFillAlpha=false
      outlineMaterial=null
      outlineSize=0
      outlineColor=<rgba 0 0 0 1>
     }
    }
    shadowOffset=<0 0>
   }
   slot {class=OverlaySlot, properties {horizontalAlignment=HAlign_Fill, verticalAlignment=VAlign_Center, padding=<ltrb 14 4 44 4>}}
  }
  WBP_UI_DownloadProgress {
   widgetClassPath=/FortUILibrary/UIKit/ProgressBar/WBP_UI_DownloadProgress
   bIsVariable=true
   parent=Overlay_37
   widget=WBP_UI_DownloadProgress
   properties {widgetSize=<32 32>}
   slot {class=OverlaySlot, properties {horizontalAlignment=HAlign_Right, verticalAlignment=VAlign_Center, padding=<ltrb 0 3 5 3>}}
  }
 ]
}
