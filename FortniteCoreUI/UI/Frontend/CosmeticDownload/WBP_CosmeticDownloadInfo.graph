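# Frontend cosmetic download notice widget that wraps a single UIKit info box with preview text warning that temporary items may show until downloads finish.
# PreConstruct pushes the exposed brightness, corner radius and material instance variables to the info box's backplate.
# It also overrides the padding on the box's horizontal layout slot.
# SetInfoBoxVisibility takes a bool to toggle the widget between collapsed and self-hit-test-invisible.
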WidgetBlueprint /FortniteCoreUI/UI/Frontend/CosmeticDownload/WBP_CosmeticDownloadInfo
graphs [
 EventGraph {nodes [
   Event PreConstruct()
   n1 = Get WBP_UIKit_Backplate!(self:WBP UIKit Info Box Object Reference=Get WBP_UIKit_InfoBox() -> (WBP_UIKit_InfoBox:WBP UIKit Info Box Object Reference)) -> (WBP_UIKit_Backplate:WBP UIKit Backplate Object Reference)
   > SetBrightness!(
    execute
    self:WBP UIKit Backplate Base Object Reference=@n1
    Brightness:E_UI_BackplateBrightness Enum=Get Brightness() -> (Brightness:E_UI_BackplateBrightness Enum=NewEnumerator0)
   ) -> (then)
   > SetCornerRadius!(
    execute
    self:WBP UIKit Backplate Base Object Reference=@n1
    Corner Radius:E_UI_BackplateCornerRadius Enum=Get Corner Radius() -> (Corner Radius:E_UI_BackplateCornerRadius Enum=NewEnumerator0)
   ) -> (then)
   > SetBackplateMaterial!(
    execute
    self:WBP UIKit Backplate Base Object Reference=@n1
    Material Instance:Material Instance Object Reference=Get Material Instance() -> (Material Instance:Material Instance Object Reference)
   ) -> (then)
   n2 = Get WBP_UIKit_InfoBox()
   > SetPadding!(
    execute
    self:Overlay Slot Object Reference=SlotAsOverlaySlot(Widget=Get HorizontalBox!(self:WBP UIKit Info Box Object Reference=@n2) -> (HorizontalBox:Horizontal Box Object Reference))
    InPadding_Left:float=10
    InPadding_Top:float=20
    InPadding_Right:float=5
    InPadding_Bottom:float=20
   ) -> (then)
  ]}
 SetInfoBoxVisibility {nodes [
   n1 = SetInfoBoxVisibility!() -> (then, ShouldBeCollapsed:bool=~) {is=title K2Node_FunctionEntry}
   > SetVisibility(InVisibility=Select(Option 0=SelfHitTestInvisible, Option 1=Collapsed, Index=@n1))
  ]}
]
variables [
 Brightness {type=E_UI_BackplateBrightness, default=Dark}
 Corner Radius {type=E_UI_BackplateCornerRadius, default=Backplate}
 Material Instance {type=MaterialInstance, default=/FortUILibrary/Systems/Backplate/MaterialInstances/MI_UIKit_Backplate}
]
widgets {
 info {
  inheritedWidgetCount=0
  parentClass=&UserWidget
  rootWidgetClass=/FortUILibrary/UIKit/InfoBox/WBP_UIKit_InfoBox.WBP_UIKit_InfoBox_C
  widgetCount=1
 }
 widgets [WBP_UIKit_InfoBox {
   widgetClassPath=/FortUILibrary/UIKit/InfoBox/WBP_UIKit_InfoBox
   bIsVariable=true
   widget=WBP_UIKit_InfoBox
   properties {previewText=Temporary items may show until your items are downloaded., showIcon=false, use Preview Text=true}
  }]
}
