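# Modal dialog widget that offers the bulk cosmetic download, wrapping the shared UIKit internal dialog.
# InitDialog stores the cosmetic download view model and passes it to that dialog to configure its widgets.
# Construct requests the download size from FortDownloadSystemVM; ReplaceDownloadButtonText formats a size into the first button's "Download" label.
# Button 0 requests installation unless already installed, and both buttons 0 and 1 pop the modal.
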WidgetBlueprint /FortniteCoreUI/UI/Frontend/CosmeticDownload/WBP_DialogModal_CosmeticDownload
graphs [
 EventGraph {nodes [
   n1 = OnDialogButtonClicked!() -> (OutputDelegate, then, ButtonIndex:int) {is=title K2Node_CustomEvent}
   > n2 = Switch on Int!(execute:Exec, Selection:int=@n1.ButtonIndex) -> (Default:Exec, 0:Exec, 1:Exec)
   n3 = Get FortDownloadSystemVM() -> (FortDownloadSystemVM:Fort Download System VM Object Reference)
   n4 = Branch(execute=@n2.0, Condition=NOT Boolean(A=GetIsBulkCosmeticInstalled!(self:Fort Download System VM Object Reference=@n3) -> (ReturnValue:bool)))
   > n5 = RequestBulkCosmeticInstall(self:Fort Download System VM Object Reference=Get FortDownloadSystemVM())
   PopCurrentModal!(execute=[@n5.then, @n2.1, @n4.else]) -> (then)
   Event OnInitialized()
   > Bind Event to Dialog Button Clicked Dispatcher(
    self:WBP UIKit Dialog Internal Object Reference=Get WBP_UIKit_Dialog_Internal() -> (WBP_UIKit_Dialog_Internal:WBP UIKit Dialog Internal Object Reference)
    Delegate=Create Event()
   ) {is=title K2Node_AddDelegate}
   Event Construct()
   > RequestBulkCosmeticDownloadSize!(execute, self:Fort Download System VM Object Reference=Get FortDownloadSystemVM(), Callback:Delegate=Create Event()) -> (then)
  ]}
 InitDialog {nodes [
   n1 = InitDialog!() -> (then, CosmeticDownloadDialogVM:Dialog VM Cosmetic Download Object Reference) {is=title K2Node_FunctionEntry}
   > n2 = Set DialogVM(DialogVM:Dialog VM Cosmetic Download Object Reference=@n1) -> (Output_Get:Dialog VM Cosmetic Download Object Reference)
   n3 = Get WBP_UIKit_Dialog_Internal()
   > DetermineWidgetsUsedBasedOnVM!(execute, self:WBP UIKit Dialog Internal Object Reference=@n3, DialogVM:UI Kit Dialog Viewmodel Object Reference=@n2) -> (then)
   > SetUIKitDialogViewModel!(execute, self:WBP UIKit Dialog Internal Object Reference=@n3, Viewmodel:UI Kit Dialog Viewmodel Object Reference=@n2) -> (then)
  ]}
 ReplaceDownloadButtonText {nodes [
   n1 = ReplaceDownloadButtonText!() -> (then, DownloadSize:Integer64) {is=title K2Node_FunctionEntry}
   n2 = "Get (a copy)"(Array:Array of WBP UIKit Button CTA Base Object References=Get ButtonWidgets!(self:WBP UIKit Dialog Internal Object Reference=Get WBP_UIKit_Dialog_Internal()) -> (ButtonWidgets:Array of WBP UIKit Button CTA Base Object References)) -> (Output:WBP UIKit Button CTA Base Object Reference)
   > SetText!(
    execute
    self:Fort CTAButton Object Reference=@n2
    ButtonText:"Text (by ref)"=Format Text!(
     Format="LOCTEXT(\"Download ({Text})\")"
     Text:Text=AsMemory!(
      NumBytes:Integer64=@n1
      UnitStandard:EMemoryUnitStandard Enum=SI
      bUseGrouping:bool=true
      MinimumIntegralDigits:int=1
      MaximumIntegralDigits:int=324
      MinimumFractionalDigits:int=2
      MaximumFractionalDigits:int=2
     ) -> (ReturnValue:Text)
    ) -> (Result)
   ) -> (then)
   > Set Array Elem!(
    execute
    TargetArray:Array of WBP UIKit Button CTA Base Object References=Get ButtonWidgets(self=Get WBP_UIKit_Dialog_Internal())
    Index:int=0
    Item:"WBP UIKit Button CTA Base Object Reference (by ref)"=@n2
    bSizeToFit:bool=false
   ) -> (then)
  ]}
]
variables [DialogVM {type=DialogVM_CosmeticDownload_C, default=null}]
widgets {
 info {
  inheritedWidgetCount=0
  parentClass=&CommonActivatableWidget
  rootWidgetClass=/FortUILibrary/Systems/Dialog/WBP_UIKit_Dialog_Internal.WBP_UIKit_Dialog_Internal_C
  widgetCount=1
 }
 widgets [WBP_UIKit_Dialog_Internal {widgetClassPath=/FortUILibrary/Systems/Dialog/WBP_UIKit_Dialog_Internal, bIsVariable=true, widget=WBP_UIKit_Dialog_Internal}]
}
